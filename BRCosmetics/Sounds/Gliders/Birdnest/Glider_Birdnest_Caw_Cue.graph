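# Sound cue for the Birdnest glider cosmetic's bird caw.
# A Random node picks one of three caw wave variants with equal weight.
# Playback uses the player glider open/close attenuation and the glider landing sound class, at slightly reduced volume.
# A concurrency override caps simultaneous instances at two and blocks new ones beyond that.
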SoundCue /BRCosmetics/Sounds/Gliders/Birdnest/Glider_Birdnest_Caw_Cue
root=Random_0
nodes [
 {id=Random_0, children [WavePlayer_0,WavePlayer_1,WavePlayer_2], properties {weights [1,1,1]}}
 {id=WavePlayer_0, properties {soundWaveAssetPtr=&/Glider_Birdnest_Caw}}
 {id=WavePlayer_1, properties {soundWaveAssetPtr=&/Glider_Birdnest_Caw_02}}
 {id=WavePlayer_2, properties {soundWaveAssetPtr=&/Glider_Birdnest_Caw_03}}
]
properties {
 attenuationSettings=/Game/Sounds/Attenuation/Fort_Player_Glider_OpenClose_Athena_Att
 bOverrideConcurrency=true
 maxDistance=10000
 soundClassObject=&Fort_Gliders_Land
 volumeMultiplier=0.7
 concurrencyOverrides {maxCount=2, resolutionRule=PreventNew}
}
