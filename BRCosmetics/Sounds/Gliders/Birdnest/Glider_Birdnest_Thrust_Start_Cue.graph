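# Sound cue for the start of thrust on the Birdnest glider cosmetic, played as a wing flap.
# A Random node picks one of six equally weighted wing flap waves and feeds a Modulator that slightly varies the volume on each play.
# The cue plays at reduced volume through the Fort_Gliders_Land sound class, with shared glider thrust attenuation and concurrency settings.
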SoundCue /BRCosmetics/Sounds/Gliders/Birdnest/Glider_Birdnest_Thrust_Start_Cue
root=Modulator_0
nodes [
 {id=Modulator_0, children [Random_1], properties {volumeMax=1, volumeMin=0.85}}
 {
  id=Random_1
  children [WavePlayer_12, WavePlayer_13, WavePlayer_14, WavePlayer_15, WavePlayer_16, WavePlayer_17]
  properties {weights [1,1,1,1,1,1]}
 }
 {id=WavePlayer_12, properties {soundWaveAssetPtr=&/Glider_Birdnest_Wing_Flap_01}}
 {id=WavePlayer_13, properties {soundWaveAssetPtr=&/Glider_Birdnest_Wing_Flap_02}}
 {id=WavePlayer_14, properties {soundWaveAssetPtr=&/Glider_Birdnest_Wing_Flap_03}}
 {id=WavePlayer_15, properties {soundWaveAssetPtr=&/Glider_Birdnest_Wing_Flap_04}}
 {id=WavePlayer_16, properties {soundWaveAssetPtr=&/Glider_Birdnest_Wing_Flap_05}}
 {id=WavePlayer_17, properties {soundWaveAssetPtr=&/Glider_Birdnest_Wing_Flap_06}}
]
properties {
 attenuationSettings=&Fort_Player_Glider_Thrust_Athena_Att
 maxDistance=12500
 soundClassObject=&Fort_Gliders_Land
 volumeMultiplier=0.3
 concurrencySet [&Athena_Glider_Thrust_Conc]
}
